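# Niagara effect that puffs glowing green-cyan spores from the Shiitake Shaolin character's face accessory, apparently on jump.
# BitsWorld emits a burst plus tapering stream of curl-noise sprites, suppressed while driving or skydiving.
# SmokeLocal adds an ungated burst of randomly flipped SubUV smoke puffs, with its rate spawner disabled.
# Both spawn on a ring under gravity and drag, with user bools boosting front-end preview alpha and setting a head-hidden visibility tag.
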NiagaraSystem /Game/Characters/Player/Male/Medium/Bodies/M_MED_Shiitake_Shaolin/FX/NS_Shiitake_Shaolin_FaceAcc_Spores_Jump
dependencies {
 usedDataInterfaces [&NiagaraDataInterfaceCurve, &NiagaraDataInterfaceArrayFloat, &NiagaraDataInterfaceColorCurve]
 usedDynamicInputs []
 usedModules [&SystemState]
}
emitters {
 BitsWorld {
  emitterUpdateScript [
   &EmitterState()
   &SpawnRate(
    Spawn Probability=&MakeCustomFloatFromBool(Bool=&Boolean_OrOperation(Bool1=$User.IsDriving, Bool2=$User.IsSkydiving), True Float=0, False Float=1)
    SpawnRate=&FloatFromCurve(
     FloatCurve=DataInterface() {propertyValues {Curve {keys [{interpMode=RCIM_Cubic, tangentMode=RCTM_User, time=0, value=1}, {interpMode=RCIM_Cubic, tangentMode=RCTM_User, time=1, value=0}]}}}
     CurveIndex=$Emitter.NormalizedLoopAge
     Scale Curve=10
    )
    Use Spawn Probability=true
   )
   &SpawnBurst_Instantaneous(
    Spawn Count=7
    Spawn Probability=&MakeCustomFloatFromBool(Bool=&Boolean_OrOperation(Bool1=$User.IsDriving, Bool2=$User.IsSkydiving), True Float=0, False Float=1)
    Use Spawn Probability=true
    UseLoopCountLimit=false
   )
  ]
  particleSpawnScript [
   &InitializeParticle(
    Lifetime=&RandomRangeFloat(Minimum=0.5, Maximum=1.3)
    Position Offset=<0 -10 -5>
    Sprite Rotation Mode=::NewEnumerator2
    Sprite Size Mode=::NewEnumerator1
    Uniform Sprite Size Max=7
    Uniform Sprite Size Min=2
    UsePositionOffset=false
    Write Material Random=true
   )
   &CylinderLocation(
    Cylinder Height=0
    Cylinder Radius=15
    MaskSpawnGroup=false
    Override Seed=false
    Rotation Mode:ENiagara_RotationMode=::NewEnumerator1
    Surface Only=true
    Use Endcaps In Surface Only Mode=false
    UseOwnerScale=true
   )
   &AddVelocityFromPoint(Module.Use Falloff Distance=false, Velocity Strength=&RandomRangeFloat(Minimum=100, Maximum=200))
   &AddVelocity(Velocity=&RandomRangeVector(Minimum=<-20 0 0>, Maximum=<-100 200 0>))
  ]
  particleUpdateScript [
   &ParticleState()
   &CurlNoiseForce(Noise Strength=500, Pan Noise=false)
   &GravityForce(Gravity=&RandomRangeVector(Minimum=<0 0 -200>, Maximum=<0 0 -400>))
   &Drag(Drag=5, Use Linear Drag=true, Use Rotational Drag=false)
   &SolveForcesAndVelocity(Clamp Velocity=false, Limit Acceleration=false)
   &DynamicMaterialParameters(
    Index 0 Param 1=&FloatFromCurve(FloatCurve=DataInterface() {propertyValues {Curve {keys [<at 0 1>,<at 1 0>]}}})
    Index 0 Param 2=&UniformRangedFloat(Minimum=20, Maximum=24)
    Index 0 Param 3=5
    Index 0 Param 4=&FloatFromCurve(FloatCurve=DataInterface() {propertyValues {Curve {keys [<at 0.0123 0.25>,<at 0.99 -0.3>,<at 1 -0.4>]}}})
    Module.Param0WriteEnabled=true
    Module.Param1WriteEnabled=true
    Module.Param2WriteEnabled=true
    Module.Param3WriteEnabled=true
   )
   &ScaleColor(
    Scale Alpha=&FloatFromCurve(
     FloatCurve=DataInterface() {propertyValues {Curve {keys [
         {interpMode=RCIM_Cubic, tangentMode=RCTM_User, time=0, value=1}
         {arriveTangent=-1, interpMode=RCIM_Cubic, leaveTangent=-1, time=0.5, value=0.8}
         {arriveTangent=-0.778, interpMode=RCIM_Cubic, leaveTangent=-0.778, tangentMode=RCTM_User, time=1, value=0}
        ]}}}
     Scale Curve=&MakeCustomFloatFromBool(Bool=$User.bIsFrontEndPreview, True Float=1.5, False Float=0.5)
    )
    Scale RGB=&MakeVectorFromLinearColorRGB(Color=&ColorFromCurve(ColorCurve=DataInterface() {propertyValues {
        AlphaCurve {keys [<at 0 1 RCIM_Cubic>]}
        BlueCurve {keys [
          {arriveTangent=-1.4598, interpMode=RCIM_Cubic, leaveTangent=-1.4598, time=0.0093, value=2.4462}
          {arriveTangent=-1.4598, interpMode=RCIM_Cubic, leaveTangent=-1.4598, time=1, value=1}
         ]}
        GreenCurve {keys [
          {arriveTangent=-3.1874, interpMode=RCIM_Cubic, leaveTangent=-3.1874, time=0.0093, value=4}
          {arriveTangent=-3.1874, interpMode=RCIM_Cubic, leaveTangent=-3.1874, time=1, value=0.8423}
         ]}
        RedCurve {keys [<at 0.0093 0 RCIM_Cubic>,<at 1 0 RCIM_Cubic>]}
       }}))
    ScaleA=true
    ScaleRGB=true
   )
   &ScaleSpriteSize(Uniform Curve Sprite Scale=DataInterface() {propertyValues {Curve {keys [
        {arriveTangent=-1.3333, interpMode=RCIM_Cubic, leaveTangent=-1.3333, time=0, value=1}
        {arriveTangent=-0.4, interpMode=RCIM_Cubic, leaveTangent=-0.4, time=0.3, value=0.6}
        {arriveTangent=-0.5, interpMode=RCIM_Cubic, leaveTangent=-0.5, time=0.5, value=0.8}
        {arriveTangent=-0.5, interpMode=RCIM_Cubic, leaveTangent=-0.5, time=0.7, value=0.4}
        {arriveTangent=-1.3333, interpMode=RCIM_Cubic, leaveTangent=-1.3333, time=0.9, value=0.6}
        {arriveTangent=-6, interpMode=RCIM_Cubic, leaveTangent=-6, time=1, value=0}
       ]}}})
   /SetVariables(Particles.VisibilityTag=&MakeCustomIntFromBool(Bool=$User.bisHeadHidden, True INT=1, False INT=0)) {isSetParameters=true, module=SetVariables_#1}
  ]
  rendererClasses [&NiagaraSpriteRendererProperties]
  renderers [{rendererClass=&NiagaraSpriteRendererProperties}]
 }
 SmokeLocal {
  emitterUpdateScript [
   &EmitterState()
   &SpawnRate(
    SpawnRate=&FloatFromCurve(
     FloatCurve=DataInterface() {propertyValues {Curve {keys [{interpMode=RCIM_Cubic, tangentMode=RCTM_User, time=0, value=1}, {interpMode=RCIM_Cubic, tangentMode=RCTM_User, time=1, value=0}]}}}
     CurveIndex=$Emitter.NormalizedLoopAge
     Scale Curve=5
    )
    Use Spawn Probability=false
   ) {enabled=false}
   &SpawnBurst_Instantaneous(Spawn Count=3, Use Spawn Probability=false, UseLoopCountLimit=false)
  ]
  particleSpawnScript [
   &InitializeParticle(
    Lifetime=&RandomRangeFloat(Minimum=0.5, Maximum=1)
    Position Offset=<0 -10 -5>
    Sprite Rotation Mode=::NewEnumerator2
    Sprite Size=&Multiply_Vector2D_ByFloat(
     Float=&RandomRangeFloat(Minimum=34, Maximum=40)
     Vector2D=&MakeVector2D(
      X=1
      Y=&SelectFloatFromArray(
       Array Sampling Mode=::NewEnumerator0
       Direct Array Index=?
       Float Selection Array=DataInterface() {propertyValues {FloatData [-1,1]}}
       Random Seed=0
      )
     )
    )
    Sprite Size Mode=::NewEnumerator4
    UsePositionOffset=false
    Write Material Random=true
   )
   &CylinderLocation(
    Cylinder Height=0
    Cylinder Radius=15
    MaskSpawnGroup=false
    Override Seed=false
    Rotation Mode:ENiagara_RotationMode=::NewEnumerator1
    Surface Only=true
    Use Endcaps In Surface Only Mode=false
    UseOwnerScale=true
   )
   &AddVelocityFromPoint(Module.Use Falloff Distance=false)
   &AddVelocity(Velocity=&RandomRangeVector(Minimum=<-20 0 0>, Maximum=<-100 200 0>))
  ]
  particleUpdateScript [
   &ParticleState()
   &GravityForce(Gravity=&RandomRangeVector(Minimum=<0 0 -50>, Maximum=<0 0 -200>))
   &Drag(Drag=5, Use Linear Drag=true, Use Rotational Drag=false)
   &SolveForcesAndVelocity(Clamp Velocity=false, Limit Acceleration=false)
   &DynamicMaterialParameters(
    Index 0 Param 1=&FloatFromCurve(FloatCurve=DataInterface() {propertyValues {Curve {keys [<at 0 1>,<at 1 0>]}}})
    Index 0 Param 2=&UniformRangedFloat(Minimum=20, Maximum=24)
    Index 0 Param 3=5
    Index 0 Param 4=&FloatFromCurve(FloatCurve=DataInterface() {propertyValues {Curve {keys [<at 0.0123 0.25>,<at 0.99 -0.3>,<at 1 -0.4>]}}})
    Module.Param0WriteEnabled=true
    Module.Param1WriteEnabled=true
    Module.Param2WriteEnabled=true
    Module.Param3WriteEnabled=true
   )
   &SubUVAnimation(
    Start Frame=24
    SubUV Animation Mode:ENiagara_SubUVLookupMode=::NewEnumerator1
    SubUV Curve=DataInterface() {propertyValues {Curve {keys [
        {arriveTangent=1.373, interpMode=RCIM_Cubic, leaveTangent=1.373, tangentMode=RCTM_User, time=0, value=0}
        {arriveTangent=0.3825, interpMode=RCIM_Cubic, leaveTangent=0.3825, tangentMode=RCTM_User, time=1, value=1}
       ]}}}
   )
   &ScaleColor(
    Scale Alpha=&FloatFromCurve(
     FloatCurve=DataInterface() {propertyValues {Curve {keys [
         {arriveTangent=-1.5971, interpMode=RCIM_Cubic, leaveTangent=-1.5971, tangentMode=RCTM_User, time=0, value=1}
         {arriveTangent=-0.778, interpMode=RCIM_Cubic, leaveTangent=-0.778, tangentMode=RCTM_User, time=1, value=0}
        ]}}}
     Scale Curve=&MakeCustomFloatFromBool(Bool=$User.bIsFrontEndPreview, True Float=1.5, False Float=0.05)
    )
    Scale RGB=&MakeVectorFromLinearColorRGB(Color=&ColorFromCurve(ColorCurve=DataInterface() {propertyValues {
        AlphaCurve {keys [<at 0 1 RCIM_Cubic>]}
        BlueCurve {keys [
          {arriveTangent=-1.4598, interpMode=RCIM_Cubic, leaveTangent=-1.4598, time=0.0093, value=2.4462}
          {arriveTangent=-1.4598, interpMode=RCIM_Cubic, leaveTangent=-1.4598, time=1, value=1}
         ]}
        GreenCurve {keys [
          {arriveTangent=-3.1874, interpMode=RCIM_Cubic, leaveTangent=-3.1874, time=0.0093, value=4}
          {arriveTangent=-3.1874, interpMode=RCIM_Cubic, leaveTangent=-3.1874, time=1, value=0.8423}
         ]}
        RedCurve {keys [<at 0.0093 0 RCIM_Cubic>,<at 1 0 RCIM_Cubic>]}
       }}))
    ScaleA=true
    ScaleRGB=true
   )
   &ScaleSpriteSize(Uniform Curve Sprite Scale=DataInterface() {propertyValues {Curve {keys [
        {arriveTangent=4.5969, interpMode=RCIM_Cubic, leaveTangent=4.5969, tangentMode=RCTM_User, time=0, value=0.2}
        {interpMode=RCIM_Cubic, tangentMode=RCTM_User, time=0.3, value=0.9}
       ]}}})
   /SetVariables(Particles.VisibilityTag=&MakeCustomIntFromBool(Bool=$User.bisHeadHidden, True INT=1, False INT=0)) {isSetParameters=true, module=SetVariables_#1}
  ]
  rendererClasses [&NiagaraSpriteRendererProperties]
  renderers [{rendererClass=&NiagaraSpriteRendererProperties}]
 }
}
summary {
 systemName=NS_Shiitake_Shaolin_FaceAcc_Spores_Jump
 emitters [
  {emitterName=BitsWorld, rendererClasses [&NiagaraSpriteRendererProperties]}
  {emitterName=SmokeLocal, rendererClasses [&NiagaraSpriteRendererProperties]}
 ]
 userVariables [
  User.bIsFrontEndPreview {type=NiagaraBool}
  User.bisHeadHidden {type=NiagaraBool}
  User.IsDriving {type=NiagaraBool}
  User.IsSkydiving {type=NiagaraBool}
 ]
}
